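# Anim Blueprint for the Assassin Suit backpack, adding physics-driven secondary motion to the pack, magazine, zippers, handle and guns.
# The input pose runs through Anim Dynamics on the main pack, magazine and zipper bones.
# Partial-alpha constraints then make handle and gun bones follow the main pack bone before further Anim Dynamics on all but gunD.
# A RigidBody node sits unconnected, and the event graph has no logic.
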AnimBlueprint /BRCosmetics/Accessories/FORT_Backpacks/Backpack_M_Assassin_Suit/Meshes/M_MED_Assassin_Suit_Pack_AnimBP
target_skeleton=&/M_MED_Assassin_Suit_Pack_Skeleton
graphs [
 AnimGraph {nodes [
   RigidBody() {
    bOverrideWorldGravity=true
    alphaBoolBlend {blendOption=Linear}
    componentAppliedLinearAccClamp=<300 300 300>
    componentLinearAccScale=<10 10 10>
    componentLinearVelScale=<10 10 10>
    overrideWorldGravity {z=-980}
   }
   AnimGraphNode_SubInput_0 = Input Pose()
   n1 = Anim Dynamics - Bone: dyn_backpackMd(ComponentPose=Local To Component(LocalPose=@AnimGraphNode_SubInput_0), Alpha=1, ExternalForce="0,0,0") {
    angularDampingOverride=0.9999
    angularSpringConstant=500
    bAngularSpring=true
    bGravityOverrideInSimSpace=true
    bOverrideAngularDamping=true
    bOverrideLinearDamping=true
    gravityScale=0.32
    linearDampingOverride=0.9999
    linearSpringConstant=200
    alphaBoolBlend {blendOption=Linear}
    boundBone {boneName=dyn_backpackMd}
    componentAppliedLinearAccClamp=<500 500 500>
    componentLinearAccScale=<2 2 2>
    componentLinearVelScale=<2 2 2>
    physicsBodyDefinitions [{
      sphereCollisionRadius=0
      boundBone {boneName=dyn_backpackMd}
      boxExtents=<30 30 5>
      constraintSetup {twistAxis=AxisZ, angularLimitsMax=<3 0 0>, angularLimitsMin=<-1 0 0>, angularTarget=<0 0 0>, linearAxesMin=<0 -0.1 0>}
      localJointOffset=<0 0 -5>
     }]
   }
   n2 = Anim Dynamics - Bone: dyn_sideMagazine(ComponentPose=@n1, Alpha=1, ExternalForce="0,0,0") {
    angularDampingOverride=1
    angularSpringConstant=500
    bAngularSpring=true
    bGravityOverrideInSimSpace=true
    bLinearSpring=true
    bOverrideAngularDamping=true
    bOverrideLinearDamping=true
    gravityScale=0.32
    linearDampingOverride=1
    alphaBoolBlend {blendOption=Linear}
    boundBone {boneName=dyn_sideMagazine}
    componentAppliedLinearAccClamp=<2000 2000 2000>
    componentLinearAccScale=<1 1 1>
    componentLinearVelScale=<1 1 1>
    physicsBodyDefinitions [{
      sphereCollisionRadius=0
      boundBone {boneName=dyn_sideMagazine}
      boxExtents=<10 10 10>
      constraintSetup {twistAxis=AxisZ, angularLimitsMax=<8 0 0>, angularLimitsMin=<-1 0 0>, angularTarget=<0 0 0>}
      localJointOffset=<0 0 -5>
     }]
   }
   n3 = Anim Dynamics - Bone: dyn_zipperA(ComponentPose=@n2, Alpha=1, ExternalForce="0,0,0") {
    angularDampingOverride=0.99
    angularSpringConstant=500
    bAngularSpring=true
    bGravityOverrideInSimSpace=true
    bLinearSpring=true
    bOverrideAngularDamping=true
    bOverrideLinearDamping=true
    gravityScale=0.32
    linearDampingOverride=0.9
    alphaBoolBlend {blendOption=Linear}
    boundBone {boneName=dyn_zipperA}
    componentAppliedLinearAccClamp=<5000 5000 5000>
    componentLinearAccScale=<1 1 1>
    componentLinearVelScale=<1 1 1>
    physicsBodyDefinitions [{
      sphereCollisionRadius=0
      boundBone {boneName=dyn_zipperA}
      boxExtents=<5 5 5>
      constraintSetup {twistAxis=AxisZ, angularLimitsMax=<5 40 5>, angularLimitsMin=<-120 -40 -20>, angularTarget=<0 0 0>}
      localJointOffset=<0 0 -3>
     }]
   }
   n4 = Anim Dynamics - Bone: dyn_zipperB(ComponentPose=@n3, Alpha=1, ExternalForce="0,0,0") {
    angularDampingOverride=0.99
    angularSpringConstant=500
    bAngularSpring=true
    bGravityOverrideInSimSpace=true
    bLinearSpring=true
    bOverrideAngularDamping=true
    bOverrideLinearDamping=true
    bUsePlanarLimit=false
    gravityScale=0.32
    linearDampingOverride=0.9
    alphaBoolBlend {blendOption=Linear}
    boundBone {boneName=dyn_zipperB}
    componentAppliedLinearAccClamp=<5000 5000 5000>
    componentLinearAccScale=<1 1 1>
    componentLinearVelScale=<1 1 1>
    physicsBodyDefinitions [{
      sphereCollisionRadius=0
      boundBone {boneName=dyn_zipperB}
      boxExtents=<5 5 5>
      constraintSetup {angularTargetAxis=AxisY, twistAxis=AxisZ, angularLimitsMax=<0 70 0>, angularLimitsMin=<0 -70 0>, angularTarget=<0 0 0>}
      localJointOffset=<0 0 -4>
     }]
   }
   n5 = "Constraint - dyn_backpackHandle (P)"(ComponentPose=@n4, Alpha=0.6, ConstraintWeights_0=1) {
    alphaBoolBlend {blendOption=Linear}
    boneToModify {boneName=dyn_backpackHandle}
    constraintSetup [{targetBone {boneName=dyn_backpackMd}}]
    constraintWeights [1]
   }
   n6 = "Constraint - dyn_gunA (P)"(ComponentPose=@n5, Alpha=0.5, ConstraintWeights_0=1) {
    alphaBoolBlend {blendOption=Linear}
    boneToModify {boneName=dyn_gunA}
    constraintSetup [{targetBone {boneName=dyn_backpackMd}}]
    constraintWeights [1]
   }
   n7 = "Constraint - dyn_gunB (P)"(ComponentPose=@n6, Alpha=0.8, ConstraintWeights_0=1) {
    alphaBoolBlend {blendOption=Linear}
    boneToModify {boneName=dyn_gunB}
    constraintSetup [{targetBone {boneName=dyn_backpackMd}}]
    constraintWeights [1]
   }
   n8 = "Constraint - dyn_gunC (P)"(ComponentPose=@n7, Alpha=0.3, ConstraintWeights_0=1) {
    alphaBoolBlend {blendOption=Linear}
    boneToModify {boneName=dyn_gunC}
    constraintSetup [{targetBone {boneName=dyn_backpackMd}}]
    constraintWeights [1]
   }
   n9 = "Constraint - dyn_gunD (P)"(ComponentPose=@n8, Alpha=0.8, ConstraintWeights_0=1) {
    alphaBoolBlend {blendOption=Linear}
    boneToModify {boneName=dyn_gunD}
    constraintSetup [{targetBone {boneName=dyn_backpackMd}}]
    constraintWeights [1]
   }
   n10 = Anim Dynamics - Bone: dyn_backpackHandle(ComponentPose=@n9, Alpha=1, ExternalForce="0,0,0") {
    angularDampingOverride=1
    angularSpringConstant=200
    bAngularSpring=true
    bGravityOverrideInSimSpace=true
    bLinearSpring=true
    bOverrideAngularDamping=true
    bOverrideLinearDamping=true
    gravityScale=0.35
    linearDampingOverride=1
    linearSpringConstant=200
    alphaBoolBlend {blendOption=Linear}
    boundBone {boneName=dyn_backpackHandle}
    componentAppliedLinearAccClamp=<400 400 500>
    componentLinearAccScale {z=1}
    componentLinearVelScale {z=1}
    gravityOverride {z=-980}
    physicsBodyDefinitions [{
      sphereCollisionRadius=0
      boundBone {boneName=dyn_backpackHandle}
      boxExtents=<15 15 5>
      constraintSetup {angularLimitsMax=<0 0 0>, angularLimitsMin=<0 0 0>, angularTarget=<0 0 0>, linearAxesMax=<0 0 1.5>}
      localJointOffset=<0 0 -3>
     }]
   }
   n11 = Anim Dynamics - Bone: dyn_gunA(ComponentPose=@n10, Alpha=1, ExternalForce="0,0,0") {
    angularDampingOverride=1
    angularSpringConstant=500
    bAngularSpring=true
    bGravityOverrideInSimSpace=true
    bLinearSpring=true
    bOverrideAngularBias=true
    bOverrideAngularDamping=true
    gravityScale=0.35
    linearSpringConstant=200
    alphaBoolBlend {blendOption=Linear}
    boundBone {boneName=dyn_gunA}
    componentAppliedLinearAccClamp=<200 200 200>
    componentLinearAccScale=<1 1 2>
    gravityOverride {z=-980}
    physicsBodyDefinitions [{
      sphereCollisionRadius=0
      boundBone {boneName=dyn_gunA}
      boxExtents=<30 30 10>
      constraintSetup {angularTargetAxis=AxisY, twistAxis=AxisZ, angularLimitsMax=<0 2 2>, angularLimitsMin=<0 -5 -2>, angularTarget=<0 0 0>}
      localJointOffset=<0 0 -3>
     }]
   }
   n12 = Anim Dynamics - Bone: dyn_gunB(ComponentPose=@n11, Alpha=1, ExternalForce="0,0,0") {
    angularDampingOverride=1
    angularSpringConstant=500
    bAngularSpring=true
    bGravityOverrideInSimSpace=true
    bLinearSpring=true
    bOverrideAngularBias=true
    bOverrideAngularDamping=true
    bOverrideLinearDamping=true
    gravityScale=0.35
    linearDampingOverride=0.9999
    linearSpringConstant=200
    alphaBoolBlend {blendOption=Linear}
    boundBone {boneName=dyn_gunB}
    componentAppliedLinearAccClamp=<200 200 200>
    componentLinearAccScale=<0.5 0.5 1>
    componentLinearVelScale=<1 1 1>
    gravityOverride {z=-980}
    physicsBodyDefinitions [{
      sphereCollisionRadius=0
      boundBone {boneName=dyn_gunB}
      boxExtents=<15 15 5>
      constraintSetup {angularTargetAxis=AxisY, twistAxis=AxisZ, angularLimitsMax=<0 3 2>, angularLimitsMin=<0 -5 -2>, angularTarget=<0 0 0>}
      localJointOffset=<0 0 -3>
     }]
   }
   n13 = Anim Dynamics - Bone: dyn_gunC(ComponentPose=@n12, Alpha=1, ExternalForce="0,0,0") {
    angularDampingOverride=1
    angularSpringConstant=500
    bAngularSpring=true
    bGravityOverrideInSimSpace=true
    bLinearSpring=true
    bOverrideAngularBias=true
    bOverrideAngularDamping=true
    gravityScale=0.35
    linearDampingOverride=0.9999
    linearSpringConstant=200
    alphaBoolBlend {blendOption=Linear}
    boundBone {boneName=dyn_gunC}
    componentAppliedLinearAccClamp=<50 50 50>
    componentLinearAccScale=<0.5 0.5 1>
    componentLinearVelScale=<1 1 1>
    gravityOverride {z=-980}
    physicsBodyDefinitions [{
      sphereCollisionRadius=0
      boundBone {boneName=dyn_gunC}
      boxExtents=<15 15 5>
      constraintSetup {angularTargetAxis=AxisY, twistAxis=AxisZ, angularLimitsMax=<0 4 0>, angularLimitsMin=<0 -4 0>, angularTarget=<0 0 0>}
      localJointOffset=<0 0 -3>
     }]
   }
   Output Pose(Result=Component To Local(ComponentPose=@n13)) {layerGroup=None}
  ]}
 EventGraph {nodes [
   TryGetPawnOwner()
   Event BlueprintUpdateAnimation()
  ]}
]
variables []
