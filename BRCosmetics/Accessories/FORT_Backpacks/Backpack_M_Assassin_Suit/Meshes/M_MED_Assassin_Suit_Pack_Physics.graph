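# Physics asset for the Assassin Suit backpack cosmetic, giving simple physics bodies to the pack and its dangling parts.
# Four box bodies cover the root, middle pack, gun and side magazine bones.
# The middle pack and gun constrain to the root, while the magazine chains off the middle pack.
# Each joint locks two angular axes, leaving only twist for pack and magazine and one swing axis for the gun.
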PhysicsAsset /BRCosmetics/Accessories/FORT_Backpacks/Backpack_M_Assassin_Suit/Meshes/M_MED_Assassin_Suit_Pack_Physics
bodies {
 backpack_root [{shapeType=Box, extentX=4.2197, extentY=17.5191, extentZ=27.4239, center=<0.2821 -3.0978 -7.3163>, rotation=<rot 4.7867 -89.9999 123.1675>}]
 dyn_backpackMd [{shapeType=Box, extentX=4.9361, extentY=16.6609, extentZ=28.742, center=<3.5785 -5.85 8.5524>, rotation=<rot 12.4487 91.2416 -94.6375>}]
 dyn_gunD [{shapeType=Box, extentX=2.0226, extentY=2.1024, extentZ=11.092, center=<0.113 0.4405 -0.1318>, rotation=<rot 0 -62.3376 -77.4026>}]
 dyn_sideMagazine [{shapeType=Box, extentX=2.5151, extentY=9.03, extentZ=7.2112, center=<0.4288 -0.3943 4.2592>, rotation=<rot -9.0677 92.0782 -80.6704>}]
}
constraints [
 {
  bone1Name=dyn_backpackMd
  bone2Name=backpack_root
  swing1LimitDegrees=45
  swing1Motion=Locked
  swing2LimitDegrees=45
  swing2Motion=Locked
  twistLimitDegrees=2
 }
 {
  bone1Name=dyn_gunD
  bone2Name=backpack_root
  swing1LimitDegrees=45
  swing1Motion=Locked
  swing2LimitDegrees=5
  twistLimitDegrees=45
  twistMotion=Locked
 }
 {
  bone1Name=dyn_sideMagazine
  bone2Name=dyn_backpackMd
  swing1LimitDegrees=45
  swing1Motion=Locked
  swing2LimitDegrees=45
  swing2Motion=Locked
  twistLimitDegrees=7
 }
]
